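# Niagara system for the melee swing trail of the MorningSoak Crypt pickaxe, built from animation-trail ribbons between socket index pairs.
# Toon_AnimTrail draws a main ribbon with front and back face colors tinted by system time-of-day values, while DarkStreaks_AnimTrail adds a thinner, grayed streak.
# Both widen quickly by system age, taper over particle life and fade alpha through curves multiplied by GeneralSwingFade.
# The older ToonTrail and DarkStreaks emitters are disabled.
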NiagaraSystem /BRCosmetics/Weapons/FORT_Melee/Pickaxe_MorningSoak_Crypt/FX/NS_Pickaxe_MorningSoak_Crypt_Trail
dependencies {
 usedDataInterfaces [&NiagaraDataInterfaceCurve, &NiagaraDataInterfaceColorCurve, &NiagaraDataInterfaceSocketReader]
 usedDynamicInputs [&ColorFromCurve, &FN_FrontEndFloatModifier]
 usedModules [
  &self:SystemSpawnScript.SystemScriptSource.SystemScriptGraph.NiagaraNodeAssignment_3.SetVariables_#1
  &self:SystemSpawnScript.SystemScriptSource.SystemScriptGraph.NiagaraNodeAssignment_4.SetVariables_#2
  &SystemState
  &self:SystemSpawnScript.SystemScriptSource.SystemScriptGraph.NiagaraNodeAssignment_1.SetVariables_#3
  &self:SystemSpawnScript.SystemScriptSource.SystemScriptGraph.NiagaraNodeAssignment_0.SetVariables_#4
 ]
}
emitters {
 DarkStreaks {
  bEnabled=false
  emitterSpawnScript [&FN_RibbonIDOffsetUtility()]
  emitterUpdateScript [
   &EmitterState()
   &FN_SpawnAnimTrail()
  ]
  particleSpawnScript [
   &FN_InitializeAnimTrailBoneIndecies(First Bone or Socket Index=1, Second Bone or Socket Index=2)
   &FN_InitializeAnimTrail(AnimTrailWidth=0.6, Lifetime=0.2)
   &AddVelocity(Velocity=<2000 0 0>) {enabled=false}
  ]
  particleUpdateScript [
   &FN_UpdateAnimTrail() {module=FN_UpdateAnimTrail001}
   &SolveForcesAndVelocity(Clamp Velocity=false, Limit Acceleration=false) {enabled=false}
   &ScaleRibbonWidth(Ribbon Width Scale=&Multiply_Float(
     A=&FloatFromCurve(FloatCurve=DataInterface() {propertyValues {Curve {keys [
          {arriveTangent=-0.7563, interpMode=RCIM_Cubic, leaveTangent=-0.7563, tangentMode=RCTM_User, time=0, value=1}
          {arriveTangent=-0.2429, interpMode=RCIM_Cubic, leaveTangent=-0.2429, tangentMode=RCTM_User, time=1, value=0.5}
         ]}}})
     B=&FloatFromCurve(
      FloatCurve=DataInterface() {propertyValues {Curve {keys [
          {arriveTangent=38.9397, interpMode=RCIM_Cubic, leaveTangent=38.9397, tangentMode=RCTM_User, time=0.025, value=0}
          {interpMode=RCIM_Cubic, tangentMode=RCTM_User, time=0.05, value=1}
         ]}}}
      CurveIndex=$Engine.System.Age
     )
    ))
   &ScaleColor(
    Scale Alpha=&FloatFromCurve(FloatCurve=DataInterface() {propertyValues {Curve {keys [
         {interpMode=RCIM_Cubic, tangentMode=RCTM_User, time=0.11, value=1}
         {arriveTangent=-21.7156, interpMode=RCIM_Cubic, leaveTangent=-21.7156, tangentMode=RCTM_User, time=0.22, value=0}
        ]}}})
    Scale RGB=&MakeVectorFromLinearColorRGB(Color=<rgba 0.36 0.38 0.4 1>)
    ScaleA=true
    ScaleRGB=true
   )
  ]
  rendererClasses [&NiagaraRibbonRendererProperties]
  renderers [{rendererClass=&NiagaraRibbonRendererProperties}]
 }
 DarkStreaks_AnimTrail {
  emitterSpawnScript [&FN_RibbonIDOffsetUtility()]
  emitterUpdateScript [
   &EmitterState()
   &FN_SpawnAnimTrail(Spawn Interval=0)
  ]
  particleSpawnScript [
   &FN_InitializeAnimTrailBoneIndecies(First Bone or Socket Index=1, Second Bone or Socket Index=2)
   &FN_InitializeAnimTrailBoneIndecies(AnimTrail Selection=1, First Bone or Socket Index=2, Second Bone or Socket Index=3) {enabled=false, module=FN_InitializeAnimTrailBoneIndecies001}
   &FN_InitializeAnimTrail(AnimTrailWidth=0.6, Lifetime=0.25)
   &AddVelocity(Velocity=<0 1000 0>) {enabled=false}
  ]
  particleUpdateScript [
   &SolveForcesAndVelocity(Clamp Velocity=false, Limit Acceleration=false) {enabled=false}
   &FN_UpdateAnimTrail() {module=FN_UpdateAnimTrail001}
   &ScaleColor(
    Scale Alpha=&Multiply_Float(
     A=&FloatFromCurve(FloatCurve=DataInterface() {propertyValues {Curve {keys [
          {interpMode=RCIM_Cubic, tangentMode=RCTM_User, time=0.14, value=1}
          {arriveTangent=-36.5544, interpMode=RCIM_Cubic, leaveTangent=-36.5544, tangentMode=RCTM_User, time=0.2, value=0}
         ]}}})
     B=$System.GeneralSwingFade
    )
    Scale RGB=&MakeVectorFromLinearColorRGB(Color=<rgba 0.36 0.38 0.4 1>)
    ScaleA=true
    ScaleRGB=true
   )
   &ScaleRibbonWidth(Ribbon Width Scale=&Multiply_Float(
     A=&FloatFromCurve(FloatCurve=DataInterface() {propertyValues {Curve {keys [
          {arriveTangent=-0.7563, interpMode=RCIM_Cubic, leaveTangent=-0.7563, tangentMode=RCTM_User, time=0, value=1}
          {arriveTangent=-0.2429, interpMode=RCIM_Cubic, leaveTangent=-0.2429, tangentMode=RCTM_User, time=1, value=0.5}
         ]}}})
     B=&FloatFromCurve(
      FloatCurve=DataInterface() {propertyValues {Curve {keys [
          {arriveTangent=72.4855, interpMode=RCIM_Cubic, leaveTangent=72.4855, tangentMode=RCTM_User, time=0.025, value=0}
          {interpMode=RCIM_Cubic, tangentMode=RCTM_User, time=0.05, value=1}
         ]}}}
      CurveIndex=$Engine.System.Age
     )
    ))
  ]
  rendererClasses [&NiagaraRibbonRendererProperties]
  renderers [
   {rendererClass=&NiagaraRibbonRendererProperties}
   {rendererClass=&NiagaraRibbonRendererProperties, rendererIndex=1}
  ]
 }
 ToonTrail {
  bEnabled=false
  emitterSpawnScript [&FN_RibbonIDOffsetUtility()]
  emitterUpdateScript [
   &EmitterState()
   &FN_SpawnAnimTrail()
  ]
  particleSpawnScript [
   &FN_InitializeAnimTrailBoneIndecies(First Bone or Socket Index=1, Second Bone or Socket Index=0)
   &FN_InitializeAnimTrail(Lifetime=0.3)
   &AddVelocity(Velocity=<2000 0 0>) {enabled=false}
  ]
  particleUpdateScript [
   &FN_UpdateAnimTrail() {module=FN_UpdateAnimTrail001}
   &SolveForcesAndVelocity(Clamp Velocity=false, Limit Acceleration=false) {enabled=false}
   &ScaleColor(
    Scale Alpha=&Multiply_Float(
     A=$System.GeneralSwingFade
     B=&FloatFromCurve(FloatCurve=DataInterface() {propertyValues {Curve {keys [
          {arriveTangent=0.029, interpMode=RCIM_Cubic, leaveTangent=0.029, tangentMode=RCTM_User, time=0.3, value=1}
          {arriveTangent=-2.7803, interpMode=RCIM_Cubic, leaveTangent=-2.7803, tangentMode=RCTM_User, time=1, value=0}
         ]}}})
    )
    Scale RGB=&MakeVectorFromLinearColorRGB(Color=<rgba 1 1 1 1>)
    ScaleA=true
    ScaleRGB=true
   )
   /SetVariables(Particles.ColorFrontFace:LinearColor=&MultiplyLinearColor(
     Linear Color=$Particles.Color
     Scale Factor=&MakeVector4_FromLinearColor(LinearColor=&FNTimeOfDayColor(Night Color=<rgba 0.34 0.38 0.5 1>, Morning Color=<rgba 1 0.92 0.9 1>, Day Color=<rgba 1 1 1 1>, Evening Color=<rgba 0.9 0.8205 0.81 1>))
    )) {enabled=false, isSetParameters=true}
   /SetVariables(Particles.ColorFrontFace:LinearColor=&MultiplyLinearColor(
     Linear Color=$Particles.Color
     Scale Factor=&MakeVector4FromVectorAndFloat("Vector (RGB)"=&MakeVectorFromLinearColorRGB(Color=$System.TOD), "Float (Alpha)"=1)
    )) {isSetParameters=true}
   /SetVariables(Particles.ColorBackFace:LinearColor=&MultiplyLinearColor(
     Linear Color=$Particles.Color
     Scale Factor=&MakeVector4_FromLinearColor(LinearColor=&FNTimeOfDayColor(
       Night Color=<rgba 0.3311 0.3823 0.77 1>
       Morning Color=<rgba 1 0.8717 0.86 1>
       Day Color=<rgba 0.95 0.8542 0.8455 1>
       Evening Color=<rgba 0.81 0.5278 0.5022 1>
      ))
    )) {enabled=false, isSetParameters=true}
   /SetVariables(Particles.ColorBackFace:LinearColor=&MultiplyLinearColor(
     Linear Color=$Particles.Color
     Scale Factor=&MakeVector4FromVectorAndFloat("Vector (RGB)"=&MakeVectorFromLinearColorRGB(Color=$System.TOD_BackSide), "Float (Alpha)"=1)
    )) {isSetParameters=true}
   &ScaleRibbonWidth(Ribbon Width Scale=&Multiply_Float(
     A=&FloatFromCurve(FloatCurve=DataInterface() {propertyValues {Curve {keys [
          {arriveTangent=-0.7563, interpMode=RCIM_Cubic, leaveTangent=-0.7563, tangentMode=RCTM_User, time=0, value=1}
          {arriveTangent=-0.2429, interpMode=RCIM_Cubic, leaveTangent=-0.2429, tangentMode=RCTM_User, time=1, value=0.5}
         ]}}})
     B=&FloatFromCurve(
      FloatCurve=DataInterface() {propertyValues {Curve {keys [
          {arriveTangent=28.194, interpMode=RCIM_Cubic, leaveTangent=28.194, tangentMode=RCTM_User, time=0.02, value=0}
          {interpMode=RCIM_Cubic, tangentMode=RCTM_User, time=0.04, value=1}
         ]}}}
      CurveIndex=$Engine.System.Age
     )
    ))
   /SetVariables(Particles.RibbonFacingAlt:Vector3f=&Multiply_Vector(A=$Particles.RibbonFacing, B=<-1 -1 -1>)) {isSetParameters=true}
  ]
  rendererClasses [&NiagaraRibbonRendererProperties]
  renderers [
   {rendererClass=&NiagaraRibbonRendererProperties}
   {rendererClass=&NiagaraRibbonRendererProperties, rendererIndex=1}
  ]
 }
 Toon_AnimTrail {
  emitterSpawnScript [&FN_RibbonIDOffsetUtility()]
  emitterUpdateScript [
   &EmitterState()
   &FN_SpawnAnimTrail()
  ]
  particleSpawnScript [
   &FN_InitializeAnimTrailBoneIndecies(First Bone or Socket Index=1, Second Bone or Socket Index=0)
   &FN_InitializeAnimTrailBoneIndecies(AnimTrail Selection=1, First Bone or Socket Index=2, Second Bone or Socket Index=3) {enabled=false, module=FN_InitializeAnimTrailBoneIndecies001}
   &FN_InitializeAnimTrail(Lifetime=0.3)
   &AddVelocity(Velocity=<0 1000 0>) {enabled=false}
  ]
  particleUpdateScript [
   &SolveForcesAndVelocity(Clamp Velocity=false, Limit Acceleration=false) {enabled=false}
   &FN_UpdateAnimTrail() {module=FN_UpdateAnimTrail001}
   &ScaleColor(
    Scale Alpha=&Multiply_Float(
     A=$System.GeneralSwingFade
     B=&FloatFromCurve(FloatCurve=DataInterface() {propertyValues {Curve {keys [
          {arriveTangent=0.029, interpMode=RCIM_Cubic, leaveTangent=0.029, tangentMode=RCTM_User, time=0.3, value=1}
          {arriveTangent=-2.0441, interpMode=RCIM_Cubic, leaveTangent=-2.0441, tangentMode=RCTM_User, time=1, value=0}
         ]}}})
    )
    Scale RGB=&MakeVectorFromLinearColorRGB(Color=<rgba 1 1 1 1>)
    ScaleA=true
    ScaleRGB=true
   )
   /SetVariables(Particles.ColorFrontFace:LinearColor=&MultiplyLinearColor(
     Linear Color=$Particles.Color
     Scale Factor=&MakeVector4FromVectorAndFloat("Vector (RGB)"=&MakeVectorFromLinearColorRGB(Color=$System.TOD), "Float (Alpha)"=1)
    )) {isSetParameters=true}
   /SetVariables(Particles.ColorBackFace:LinearColor=&MultiplyLinearColor(
     Linear Color=$Particles.Color
     Scale Factor=&MakeVector4FromVectorAndFloat("Vector (RGB)"=&MakeVectorFromLinearColorRGB(Color=$System.TOD_BackSide), "Float (Alpha)"=1)
    )) {isSetParameters=true}
   &ScaleRibbonWidth(Ribbon Width Scale=&Multiply_Float(
     A=&FloatFromCurve(FloatCurve=DataInterface() {propertyValues {Curve {keys [
          {arriveTangent=-0.7563, interpMode=RCIM_Cubic, leaveTangent=-0.7563, tangentMode=RCTM_User, time=0, value=1}
          {arriveTangent=-0.2429, interpMode=RCIM_Cubic, leaveTangent=-0.2429, tangentMode=RCTM_User, time=1, value=0.5}
         ]}}})
     B=&FloatFromCurve(
      FloatCurve=DataInterface() {propertyValues {Curve {keys [
          {arriveTangent=28.194, interpMode=RCIM_Cubic, leaveTangent=28.194, tangentMode=RCTM_User, time=0.02, value=0}
          {interpMode=RCIM_Cubic, tangentMode=RCTM_User, time=0.04, value=1}
         ]}}}
      CurveIndex=$Engine.System.Age
     )
    ))
   /SetVariables(Particles.RibbonFacingAlt:Vector3f=&Multiply_Vector(A=$Particles.RibbonFacing, B=<-1 -1 -1>)) {isSetParameters=true}
  ]
  rendererClasses [&NiagaraRibbonRendererProperties]
  renderers [
   {rendererClass=&NiagaraRibbonRendererProperties}
   {rendererClass=&NiagaraRibbonRendererProperties, rendererIndex=1}
   {rendererClass=&NiagaraRibbonRendererProperties, rendererIndex=2}
  ]
 }
}
summary {
 systemName=NS_Pickaxe_MorningSoak_Crypt_Trail
 emitters [
  {bEnabled=false, emitterName=ToonTrail, rendererClasses [&NiagaraRibbonRendererProperties]}
  {bEnabled=false, emitterName=DarkStreaks, rendererClasses [&NiagaraRibbonRendererProperties]}
  {emitterName=Toon_AnimTrail, rendererClasses [&NiagaraRibbonRendererProperties]}
  {emitterName=DarkStreaks_AnimTrail, rendererClasses [&NiagaraRibbonRendererProperties]}
 ]
 userVariables [
  User.bIsFrontEnd {type=NiagaraBool}
  User.bIsFrontEndPreview {type=NiagaraBool}
  User.RibbonIDOffset {type=NiagaraInt32}
 ]
}
